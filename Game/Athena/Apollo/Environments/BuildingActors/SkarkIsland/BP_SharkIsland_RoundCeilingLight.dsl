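; Blueprint for a round ceiling light building actor in the Shark Island environment set.
; Its construction script runs the parent logic, then destroys the SpotLight component when the killSpotlight flag in the Lighting category is set.
; This lets a placed instance drop its spotlight.
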(fn ConstructionScript
  (|CallParentFunction)
  (if (Variables|Lighting|GetkillSpotlight)
    (Components|DestroyComponent (Variables|Default|GetSpotLight))))
